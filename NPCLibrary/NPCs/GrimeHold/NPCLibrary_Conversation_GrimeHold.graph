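# Conversation database for the GrimeHold NPC, Blinky Fishstick, defining its greeting and the menus it leads to.
# An entry point feeds one speech node with a default blink line and contextual variants for first-ever meeting, match conversation count, and wearing the TeriyakiFish outfit.
# Curve rows set priority for the first-meeting and outfit lines.
# The speech then links to the shared NPC library Quests and BasicServices conversation entries.
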ConversationDatabase /NPCLibrary/NPCs/GrimeHold/NPCLibrary_Conversation_GrimeHold
entry_points [{tag=Conversation.Entry.GrimeHold, destinations [#1]}]
nodes [
 ConversationEntryPointNode_7 {
  id=#1
  class=ConversationEntryPointNode
  outputs [#2]
  properties {entryTag {tagName=Conversation.Entry.GrimeHold}}
 }
 Speech_C_4 {
  id=#2
  class=Speech_C
  outputs [#3, #4]
  properties {
   nodeName=""
   generalConfig {
    defaultMessage=(Blinky Fishstick triple-blinks slowly.)
    contextualMessages [
     {
      message=(Blinky Fishstick gives you a glowing hello!)
      contextRequirements [{requirement {class=FortControllerRequirement_IsFirstEverConversationWithNPC}}]
      priority {curve {rowName=Tandem.Speech.Priority.FirstEverInteraction}}
     }
     {
      message=(Blinky Fishstick triple-blinks slowly.)
      contextRequirements [{requirement {class=FortControllerRequirement_CheckNPCMatchConversationCount, properties {nPCUniqueIdTag {tagName=AISpawnerData.Type.Tandem.GrimeHold}}}}]
      priority {curve {rowName=None}}
     }
     {
      message=(Blinky Fishstick subtly tries to spot your third eye.)
      contextRequirements [{requirement {
         class=FortControllerRequirement_HasCID
         properties {allowedCIds [&CID_315_Athena_Commando_M_TeriyakiFish]}
        }}]
      priority {curve {rowName=Tandem.Speech.Priority.OutfitReaction}}
     }
    ]
   }
  }
 }
 ConversationLinkNode_0 {id=#4, class=ConversationLinkNode, properties {remoteEntryTag {tagName=Conversation.Entry.BasicServices}}}
 ConversationLinkNode_1 {id=#3, class=ConversationLinkNode, properties {remoteEntryTag {tagName=Conversation.Entry.NPCLibrary.Quests}}}
]
